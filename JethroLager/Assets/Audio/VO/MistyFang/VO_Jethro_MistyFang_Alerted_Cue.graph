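# Sound cue for the Jethro MistyFang character's "Alerted" voice lines.
# A Random node picks one of five equally weighted wave players, each holding a different recorded take.
# Playback has boosted volume, uses the shared NPC alerted-VO attenuation and non-critical enemy VO sound class, and is limited by the SipTea chatter concurrency set.
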SoundCue /JethroLager/Assets/Audio/VO/MistyFang/VO_Jethro_MistyFang_Alerted_Cue
root=Random_1
nodes [
 {id=Random_1, children [WavePlayer_10, WavePlayer_11, WavePlayer_12, WavePlayer_13, WavePlayer_14], properties {weights [1,1,1,1,1]}}
 {id=WavePlayer_10, properties {soundWaveAssetPtr=/JethroLager/Assets/Audio/VO/MistyFang/Waves/VO_Jethro_MistyFang_Alerted_010}}
 {id=WavePlayer_11, properties {soundWaveAssetPtr=/JethroLager/Assets/Audio/VO/MistyFang/Waves/VO_Jethro_MistyFang_Alerted_020}}
 {id=WavePlayer_12, properties {soundWaveAssetPtr=/JethroLager/Assets/Audio/VO/MistyFang/Waves/VO_Jethro_MistyFang_Alerted_030}}
 {id=WavePlayer_13, properties {soundWaveAssetPtr=/JethroLager/Assets/Audio/VO/MistyFang/Waves/VO_Jethro_MistyFang_Alerted_080}}
 {id=WavePlayer_14, properties {soundWaveAssetPtr=/JethroLager/Assets/Audio/VO/MistyFang/Waves/VO_Jethro_MistyFang_Alerted_090}}
]
properties {
 attenuationSettings=/Game/Sounds/Fort_Human_VO_BR/NPCs/Shared/NPC_VO_Alerted_Att
 maxDistance=12500
 soundClassObject=/Game/Sounds/FortSoundClassesAndModes/Fort_VO_Enemy_NonCritical
 volumeMultiplier=2.1188
 concurrencySet [/SipTea/Sounds/VO/SipTea_Chatter_Conc]
}
