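# Sound cue for the Misty Fang voice set in JethroLager, playing a voice line for when the NPC loses its target.
# A Random node picks one of six equally weighted LostTarget wave recordings.
# It uses shared alerted-NPC attenuation, the non-critical enemy VO sound class and a chatter concurrency set, with a raised volume multiplier.
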SoundCue /JethroLager/Assets/Audio/VO/MistyFang/VO_Jethro_MistyFang_LostTarget_Cue
root=Random_1
nodes [
 {
  id=Random_1
  children [WavePlayer_6, WavePlayer_7, WavePlayer_8, WavePlayer_9, WavePlayer_10, WavePlayer_11]
  properties {weights [1,1,1,1,1,1]}
 }
 {id=WavePlayer_10, properties {soundWaveAssetPtr=/JethroLager/Assets/Audio/VO/MistyFang/Waves/VO_Jethro_MistyFang_LostTarget_052}}
 {id=WavePlayer_11, properties {soundWaveAssetPtr=/JethroLager/Assets/Audio/VO/MistyFang/Waves/VO_Jethro_MistyFang_LostTarget_060}}
 {id=WavePlayer_6, properties {soundWaveAssetPtr=/JethroLager/Assets/Audio/VO/MistyFang/Waves/VO_Jethro_MistyFang_LostTarget_020}}
 {id=WavePlayer_7, properties {soundWaveAssetPtr=/JethroLager/Assets/Audio/VO/MistyFang/Waves/VO_Jethro_MistyFang_LostTarget_030}}
 {id=WavePlayer_8, properties {soundWaveAssetPtr=/JethroLager/Assets/Audio/VO/MistyFang/Waves/VO_Jethro_MistyFang_LostTarget_040}}
 {id=WavePlayer_9, properties {soundWaveAssetPtr=/JethroLager/Assets/Audio/VO/MistyFang/Waves/VO_Jethro_MistyFang_LostTarget_051}}
]
properties {
 attenuationSettings=/Game/Sounds/Fort_Human_VO_BR/NPCs/Shared/NPC_VO_Alerted_Att
 maxDistance=12500
 soundClassObject=/Game/Sounds/FortSoundClassesAndModes/Fort_VO_Enemy_NonCritical
 volumeMultiplier=1.4125
 concurrencySet [/SipTea/Sounds/VO/SipTea_Chatter_Conc]
}
